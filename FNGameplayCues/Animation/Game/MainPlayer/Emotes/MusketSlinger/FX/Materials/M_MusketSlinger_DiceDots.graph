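# Masked, lit material for the dice dots in the MusketSlinger emote effects, flagged for skeletal and Niagara meshes.
# UV math builds a procedural circular gradient and mask, so no textures are sampled.
# The gradient forms a normal-like vector with inverted XY that blends toward other values near the center, suggesting a recessed pip.
# A powered, smoothstepped copy of the gradient tints a golden glow through MF_FX_EmissiveAdjustments, alongside a dark grey constant.
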Material /FNGameplayCues/Animation/Game/MainPlayer/Emotes/MusketSlinger/FX/Materials/M_MusketSlinger_DiceDots
graph {
 properties {
  bUsedWithMeshDeformer=true
  bUsedWithNiagaraMeshParticles=true
  bUsedWithSkeletalMesh=true
  blendMode=BLEND_Masked
  pixelDepthOffsetMode=PDOM_Legacy
  usedShadingModels=Default Lit
 }
 nodes [
  n1 = ConstantBiasScale(None=TextureCoordinate()) {bias=-0.5, scale=2}
  n2 = OneMinus(None=Saturate(None=DotProduct(A=@n1, B=@n1)))
  n3 = SquareRoot(None=@n2)
  n4 = Ceil(None=@n3)
  n5 = SmoothStep(Value=@n2) {constMax=0.55, constMin=0.4}
  n6 = LinearInterpolate(A=OneMinus(None=@n1), B=Constant2Vector() {g=0.5, r=0.5}, Alpha=@n5)
  n7 = OneMinus(None=@n3)
  AppendVector(A=Multiply(A=@n4, B=Saturate(None=@n6)), B=Add(A=LinearInterpolate(A=@n7, Alpha=@n5), B=OneMinus(None=@n4)))
  Constant3Vector() {constant=<rgba 0.0319 0.0319 0.0319 1>}
  n8 = Power(Base=@n2)
  MaterialFunctionCall(Emissive=Multiply(A=Constant3Vector() {constant=<rgba 2 1.2292 0.1037 1>}, B=SmoothStep(Value=@n8))) {materialFunction=&MF_FX_EmissiveAdjustments}
 ]
}
