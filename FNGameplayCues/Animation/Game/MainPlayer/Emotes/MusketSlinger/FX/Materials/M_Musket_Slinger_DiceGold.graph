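# Default Lit gold material for dice in the Musket Slinger emote effects, flagged for Niagara mesh particles, skeletal meshes, Nanite and mesh deformers.
# A Fresnel function blends two warm gold color parameters, ColorMain and Color, for view-dependent tint.
# Untextured surface values come from plain constants and a scalar roughness parameter, misspelled "Shine Roughnmess".
# The graph does not show which material outputs these nodes feed.
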Material /FNGameplayCues/Animation/Game/MainPlayer/Emotes/MusketSlinger/FX/Materials/M_Musket_Slinger_DiceGold
graph {
 properties {
  bUsedWithMeshDeformer=true
  bUsedWithNanite=true
  bUsedWithNiagaraMeshParticles=true
  bUsedWithSkeletalMesh=true
  pixelDepthOffsetMode=PDOM_Legacy
  usedShadingModels=Default Lit
 }
 nodes [
  Constant() {r=1}
  Constant() {r=0.5}
  LinearInterpolate(
   A=VectorParameter() {parameterName=ColorMain, defaultValue=<rgba 1 0.5683 0.2174 1>}.RGB
   B=VectorParameter() {parameterName=Color, defaultValue=<rgba 0.5885 0.3345 0.1279 1>}.RGB
   Alpha=MaterialFunctionCall(Brightness=Constant() {r=1}, Exponent=Constant() {r=2}) {materialFunction=&MF_Fresnel_FX}.Fresnel
  )
  ScalarParameter() {defaultValue=0.5, parameterName=Shine Roughnmess}
 ]
}
